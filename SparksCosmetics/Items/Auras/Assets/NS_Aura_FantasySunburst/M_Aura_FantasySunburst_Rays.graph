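# Translucent, unlit, two-sided Niagara sprite material that draws wavy rays for the Fantasy Sunburst aura cosmetic.
# A time-panned sine offsets the UVs to shape a procedural, tapered ray mask with no textures, blending between hard and soft edges.
# Opacity combines that mask with particle alpha, a tip fade and depth fade.
# Emissive splits two colors along the waving centerline, applies a base-to-tip brightness gradient and particle color, then passes through MF_FX_EmissiveAdjustments.
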Material /SparksCosmetics/Items/Auras/Assets/NS_Aura_FantasySunburst/M_Aura_FantasySunburst_Rays
graph {
 properties {
  bAutomaticallySetUsageInEditor=false
  bCastRayTracedShadows=false
  bUsedWithNiagaraSprites=true
  blendMode=BLEND_Translucent
  shadingModel=MSM_Unlit
  twoSided=true
  usedShadingModels=Unlit
 }
 nodes [
  n1 = ParticleColor()
  MaterialExpressionTextureCoordinate_2 = TextureCoordinate()
  n2 = Convert(None=@MaterialExpressionTextureCoordinate_2) {
   nodeName=Break Float2
   convertInputs [{
     type=Vector2
     expressionInput {
      expression=&self:MaterialExpressionTextureCoordinate_2
      inputName=None
      mask=0
      maskA=0
      maskB=0
      maskG=0
      maskR=0
      outputIndex=0
     }
    }]
   convertOutputs [{type=Scalar}, {type=Scalar}]
  }
  n3 = OneMinus(None=@n2)
  n4 = Multiply(A=@n3) {constB=2.5}
  n5 = Add(
   A=Multiply(A=ScalarParameter() {defaultValue=3, group=Wave Controls, parameterName=Wave Count}, B=@n3) {constB=3}
   B=Multiply(A=Time(), B=ScalarParameter() {defaultValue=1, group=Wave Controls, parameterName=Wave Pan Speed})
  )
  n6 = ScalarParameter() {defaultValue=0.2, group=Wave Controls, parameterName=Wave Scale}
  n7 = Add(A=@n2, B=Multiply(A=Sine(None=@n5), B=Multiply(A=@n6) {constB=-0.5}))
  n8 = Add(A=@n7) {constA=-0.5, constB=-0.5}
  n9 = Add(
   A=ScalarParameter() {defaultValue=-0.25, group=Wave Controls, parameterName=Wave Edge Padding}
   B=Add(A=Multiply(A=@n6) {constB=4}) {constA=2, constB=2}
  )
  n10 = SmoothStep(Min=Power(Base=Min(A=@n4)) {constExponent=0.5}, Value=Abs(None=Multiply(A=@n8, B=@n9) {constB=2.3})) {constMax=0}
  n11 = Constant2Vector() {g=0.5}
  n12 = Constant2Vector() {g=1}
  n13 = ScalarParameter() {defaultValue=0.35, group=Opacity, parameterName="Use Hard Edges (0-1)"}
  MaterialExpressionTextureCoordinate_4 = TextureCoordinate()
  n14 = Convert(None=@MaterialExpressionTextureCoordinate_4) {
   nodeName=Break Float2
   convertInputs [{
     type=Vector2
     expressionInput {
      expression=&self:MaterialExpressionTextureCoordinate_4
      inputName=None
      mask=0
      maskA=0
      maskB=0
      maskG=0
      maskR=0
      outputIndex=0
     }
    }]
   convertOutputs [{type=Scalar}, {type=Scalar}]
  }
  n15 = LinearInterpolate(
   A=Ceil(None=@n10)
   B=MaterialFunctionCall(Input=@n10, Input Low=@n11.R, Input High=@n11.R, Target Low=@n12.R, Target High=@n12.R) {materialFunction=&RemapValueRange}.Result
   Alpha=Add(A=OneMinus(None=@n13), B=Multiply(A=@n13, B=@n14) {constB=0.25}) {constB=0.75}
  )
  n16 = ScalarParameter() {defaultValue=0.95, group=Opacity, parameterName="Fade Tip (0-1)"}
  n17 = Power(Base=OneMinus(None=Multiply(A=@n14, B=@n16) {constB=0.95}), Exp=Constant() {r=1.5})
  DepthFade(
   Opacity=Multiply(A=Multiply(A=@n1.A, B=@n15), B=Saturate(None=@n17))
   FadeDistance=ScalarParameter() {defaultValue=5, parameterName=Depth Fade Distance}
  )
  n18 = Add(A=@n7) {constB=-0.5}
  n19 = Constant2Vector() {g=0.1, r=-0.2}
  n20 = Constant2Vector() {g=1}
  MaterialFunctionCall(Input=@n18, Input Low=@n19.R, Input High=@n19.R, Target Low=@n20.R, Target High=@n20.R) {materialFunction=&RemapValueRange}
  MaterialExpressionTextureCoordinate_3 = TextureCoordinate()
  n21 = Convert(None=@MaterialExpressionTextureCoordinate_3) {
   nodeName=Break Float2
   convertInputs [{
     type=Vector2
     expressionInput {
      expression=&self:MaterialExpressionTextureCoordinate_3
      inputName=None
      mask=0
      maskA=0
      maskB=0
      maskG=0
      maskR=0
      outputIndex=0
     }
    }]
   convertOutputs [{type=Scalar}, {type=Scalar}]
  }
  n22 = ScalarParameter() {defaultValue=2, group=Color, parameterName=Color Mult Base}
  n23 = ScalarParameter() {defaultValue=0.5, group=Color, parameterName=Color Mult Tip}
  n24 = Add(A=Multiply(A=OneMinus(None=@n21), B=Subtract(A=@n22, B=@n23)) {constB=1.5}, B=@n23) {constB=0.5}
  n25 = ScalarParameter() {defaultValue=1, group=Color, parameterName=Use Split Color}
  n26 = Ceil(None=@n18)
  n27 = LinearInterpolate(
   A=VectorParameter() {group=Color, parameterName=Color A, defaultValue=<rgba 1 0.6925 0 1>}.RGB
   B=VectorParameter() {group=Color, parameterName=Color B, defaultValue=<rgba 1.18 0.2955 0 1>}.RGB
   Alpha=Saturate(None=Multiply(A=@n25, B=@n26))
  )
  MaterialFunctionCall(Emissive=Multiply(A=Multiply(A=@n24, B=@n27), B=@n1.RGB)) {materialFunction=&MF_FX_EmissiveAdjustments}
 ]
}
