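# Niagara aura cosmetic that draws a fantasy sunburst behind a character, with glows on the body.
# A rotating backer and two interleaved rings of sine-swaying rays and beams follow a system-computed center, while twinkling stars burst outward.
# BodyGlow pins sprites to skeletal bones or between bone pairs.
# User flags adjust large-body sizing and front-end brightness, time-of-day scalars dim most layers, and the Minimal emitter is disabled.
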NiagaraSystem /SparksCosmetics/Items/Auras/Assets/NS_Aura_FantasySunburst/NS_Aura_FantasySunburst
dependencies {
 usedDataInterfaces [
  &NiagaraDataInterfaceArrayFloat
  &NiagaraDataInterfaceCurve
  &NiagaraDataInterfaceArrayInt32
  &NiagaraDataInterfaceArrayFloat3
  &NiagaraDataInterfaceColorCurve
  &NiagaraDataInterfaceSkeletalMesh
 ]
 usedDynamicInputs [&MakeVector, &MakeCustomVectorFromBool]
 usedModules [
  &self:SystemSpawnScript.SystemScriptSource.SystemScriptGraph.NiagaraNodeAssignment_0.SetVariables_#1
  &self:SystemSpawnScript.SystemScriptSource.SystemScriptGraph.NiagaraNodeAssignment_3.SetVariables_#2
  &SystemState
  &SampleSkeletalMesh
  &self:SystemSpawnScript.SystemScriptSource.SystemScriptGraph.NiagaraNodeAssignment_2.SetVariables_#3
 ]
}
emitters {
 Backer {
  emitterUpdateScript [
   &EmitterState()
   &SpawnBurst_Instantaneous(Use Spawn Probability=false, UseLoopCountLimit=true)
  ]
  particleSpawnScript [&InitializeParticle(
    Color=<rgba 1 0.1667 0 0.65>
    Lifetime=3
    Sprite Size Mode=::NewEnumerator3
    Uniform Sprite Size=&Multiply_Float(A=125, B=&MakeCustomFloatFromBool(Bool=$User.IsLargeBody, True Float=1.4, False Float=1))
    UsePositionOffset=false
    Write Material Random=true
   ) {module=InitializeParticle001}]
  particleUpdateScript [
   &ParticleState(
    Kill Particles When Lifetime Has Elapsed=false
    Let Infinitely Lived Particles Die When Emitter Deactivates=true
    Loop Particles Lifetime=true
   )
   &SpriteRotationRate(Rotation Rate=5)
   &CameraOffset(Camera Offset Amount=-20)
   &ScaleColor(
    Scale Alpha=&Multiply_Float(
     A=&FNTimeOfDayScalar(Night Scale=0.5, Morning Scale=0.8, Evening Scale=0.8)
     B=&MakeCustomFloatFromBool(Bool=$User.IsFrontEnd, True Float=0.6, False Float=1)
    )
    ScaleA=true
    ScaleRGB=true
   )
   &DynamicMaterialParameters(
    Index 0 Param 1=2
    Index 0 Param 2=1
    Index 0 Param 4=1
    Module.Index0Float0Write=true
    Module.Index0Float1Write=true
    Module.Index0Float2Write=true
    Module.Index0Float3Write=true
   )
   /SetVariables(Particles.Position=$System.CenteredPos) {isSetParameters=true}
   /SetVariables(Particles.DarkColor:LinearColor=<rgba 0.0066 0 0.0948 2>) {isSetParameters=true}
  ]
  rendererClasses [&NiagaraSpriteRendererProperties]
  renderers [
   {rendererClass=&NiagaraSpriteRendererProperties}
   {rendererClass=&NiagaraSpriteRendererProperties, rendererIndex=1}
  ]
 }
 Beams {
  emitterUpdateScript [
   &EmitterState()
   &SpawnBurst_Instantaneous(Spawn Count=11, Use Spawn Probability=false, UseLoopCountLimit=true)
  ]
  particleSpawnScript [
   &InitializeParticle(
    Color=<rgba 1 0.25 0 0.65>
    Lifetime=7
    Sprite Rotation Angle=&Add_Float(
     A=&Divide_Float(A=360, B=22)
     B=&Multiply_FloatByInt(Float=&Divide_Float(A=360, B=11), Integer=&ReturnExecIndex())
    )
    Sprite Rotation Mode=::NewEnumerator0
    Sprite Size=&Multiply_Vector2D(
     A=<8 55>
     B=&MakeCustomVector2DFromBool(Bool=$User.IsLargeBody, True Vector2D=<1.2 1.2>, False Vector2D=<1 1>)
    )
    Sprite Size Mode=::NewEnumerator4
    UsePositionOffset=false
    Write Material Random=true
   )
   &CameraOffset(Camera Offset Amount=-50) {module=CameraOffset001}
  ]
  particleUpdateScript [
   &ParticleState(Kill Particles When Lifetime Has Elapsed=false)
   &SpriteRotationRate(Rotation Rate=&Multiply_Float(A=10, B=&Sine(Normalized Angle=$Particles.NormalizedAge, Period=2, Scale=1, Bias=0)))
   /SetVariables(Particles.Position=$System.CenteredPos) {isSetParameters=true, module=SetVariables_#4}
   &ScaleColor(
    Scale Alpha=&FNTimeOfDayScalar(Night Scale=0.5, Morning Scale=0.75, Evening Scale=1)
    ScaleA=true
    ScaleRGB=true
   )
  ]
  rendererClasses [&NiagaraSpriteRendererProperties]
  renderers [{rendererClass=&NiagaraSpriteRendererProperties}]
 }
 BodyGlow {
  emitterUpdateScript [
   &EmitterState()
   &SpawnBurst_Instantaneous(Spawn Count=6, Use Spawn Probability=false, UseLoopCountLimit=true)
  ]
  particleSpawnScript [
   &PartitionParticles(Number Of Partitions=6)
   &InitializeParticle(
    Color=<rgba 1 0.5061 0.1768 0.65>
    Lifetime=&RandomRangeFloat(Minimum=3, Maximum=4)
    Sprite Rotation Mode=::NewEnumerator2
    Sprite Size Mode=::NewEnumerator3
    Uniform Sprite Size=&MakeCustomFloatFromBool(Bool=$User.IsLargeBody, True Float=30, False Float=20)
    UsePositionOffset=false
    Write Material Random=true
   ) {module=InitializeParticle001}
   /SetVariables(
    Particles.PickBone:NiagaraInt32=&SelectIntFromArray(Direct Array Index=$Particles.PartitionParticles.Partition, Int Selection Array=DataInterface() {propertyValues {IntData [1,2,3,4,3,4]}})
    Particles.SymmetricalLerp:NiagaraFloat=&Multiply_Float(
     A=/Niagara/DynamicInputs/Integer/V2/Round_Float(Float:NiagaraFloat=&Multiply_FloatByInt(Float=0.49, Integer=$Particles.PartitionParticles.Partition))
     B=0.5
    )
   ) {isSetParameters=true}
   /SetVariables(
    Particles.glowCameraOffset=-20
    Particles.glowColor:LinearColor=&MakeCustomLinearColorFromBool(
     Bool=$User.IsFrontEnd
     True Color=&LerpLinearColors(StartColor=<rgba 1 0.5061 0.1768 0.3>, EndColor=<rgba 1 0.2591 0.1768 0.15>, LerpFactor=$Particles.SymmetricalLerp)
     False Color=&LerpLinearColors(StartColor=<rgba 1 0.5061 0.1768 0.45>, EndColor=<rgba 1 0.2591 0.1768 0.15>, LerpFactor=$Particles.SymmetricalLerp)
    )
    Particles.glowScale:Vector2f=&Vector2DFromFloat(Value=&Multiply_Float(
      A=&SelectFloatFromArray(
       Direct Array Index=$Particles.PartitionParticles.Partition
       Float Selection Array=DataInterface() {propertyValues {FloatData [60,60,40,40,40,40]}}
      )
      B=&MakeCustomFloatFromBool(Bool=$User.IsLargeBody, True Float=1.3, False Float=1)
     ))
   ) {isSetParameters=true}
  ]
  particleUpdateScript [
   &ParticleState(Kill Particles When Lifetime Has Elapsed=false, Let Infinitely Lived Particles Die When Emitter Deactivates=true)
   &CameraOffset(Camera Offset Amount=5)
   &ScaleSpriteSize(
    Scale Sprite Size Mode=::NewEnumerator0
    Uniform Scale Factor=&Lerp_Float(A=1.5, B=1, Alpha=$Particles.SymmetricalLerp)
   )
   &ScaleColor(
    Scale RGB=&VectorFromFloat(Value=&Lerp_Float(A=3, B=0.75, Alpha=$Particles.SymmetricalLerp))
    ScaleA=true
    ScaleRGB=true
   )
   &SpriteRotationRate(Rotation Rate=&RandomRangeFloat(Minimum=10, Maximum=20))
   &SkeletalMeshLocation(
    Bone / Socket Index=$Particles.PickBone
    Bone Sampling Mode=::NewEnumerator4
    MaskSpawnGroup=false
    Orientation Sampling=::NewEnumerator1
    Position Sampling=::NewEnumerator2
    Sampled Position Offset=&Multiply_VectorByFloat(
     Vector=&SelectVectorFromArray(
      Direct Array Index=$Particles.PickBone
      Vector Selection Array=DataInterface() {propertyValues {FloatData [<0 0 0>, <-5 5 0>, <5 -5 0>, <0 0 0>]}}
     )
     Float=&MakeCustomFloatFromBool(Bool=$User.IsLargeBody, True Float=1, False Float=0.5)
    )
    Skeletal Mesh=$System.Skeletal Mesh
   )
   &SkeletalMeshLocation(
    Bone / Socket Index=&SelectIntFromArray(Direct Array Index=$Particles.PickBone, Int Selection Array=$System.CompanionBone)
    Bone Sampling Mode=::NewEnumerator4
    MaskSpawnGroup=false
    Orientation Sampling=::NewEnumerator1
    Position Sampling=::NewEnumerator2
    Sampled Position Offset Coordinate Space:ENiagaraSkelMeshTransforms=::NewEnumerator1
    Skeletal Mesh=$System.Skeletal Mesh
   ) {module=SkeletalMeshLocation001}
   &DynamicMaterialParameters(
    Index 0 Param 1=&SelectFloatFromArray(
     Direct Array Index=$Particles.PartitionParticles.Partition
     Float Selection Array=DataInterface() {propertyValues {FloatData [5,5,5,4.5,4.5,4,4,3,3]}}
    )
    Index 0 Param 3=1
    Module.Param0WriteEnabled=true
    Module.Param1WriteEnabled=true
    Module.Param2WriteEnabled=true
    Module.Param3WriteEnabled=true
   )
   /SetVariables(Particles.Position=&LerpPosition(
     A=$Particles.SkeletalMeshLocation.SampledPosition
     B=$Particles.SkeletalMeshLocation001.SampledPosition
     Alpha=&SelectFloatFromArray(
      Direct Array Index=$Particles.PartitionParticles.Partition
      Float Selection Array=DataInterface() {propertyValues {FloatData [0,0,0.5,0.5,0,0,0.5,0.5,0,0]}}
     )
    )) {isSetParameters=true}
   &Color(Color=&ColorFromCurve(
     ColorCurve=DataInterface() {propertyValues {
       AlphaCurve {keys [<at 0 1 RCIM_Cubic>]}
       BlueCurve {keys [<at 0 0 RCIM_Cubic>,<at 0.5 0>,<at 1 1 RCIM_Cubic>]}
       GreenCurve {keys [<at 0 0 RCIM_Cubic>,<at 0.5 1>,<at 1 0 RCIM_Cubic>]}
       RedCurve {keys [<at 0 1 RCIM_Cubic>,<at 0.5 0>,<at 1 0 RCIM_Cubic>]}
      }}
     CurveIndex=$Particles.SymmetricalLerp
    )) {enabled=false}
  ]
  rendererClasses [&NiagaraSpriteRendererProperties]
  renderers [
   {rendererClass=&NiagaraSpriteRendererProperties}
   {rendererClass=&NiagaraSpriteRendererProperties, rendererIndex=1}
   {rendererClass=&NiagaraSpriteRendererProperties, rendererIndex=2}
  ]
 }
 Minimal {
  bEnabled=false
  emitterUpdateScript [
   &EmitterState()
   &SpawnRate(Use Spawn Probability=false)
  ]
  particleSpawnScript [
   &InitializeParticle(Color Mode=::NewEnumerator0, Sprite Size Mode=::NewEnumerator3, Uniform Sprite Size=35, UsePositionOffset=false, Write Material Random=true)
   &ShapeLocation(
    MaskSpawnGroup=false
    Offset:NiagaraPosition=$System.CenteredPos
    Offset Mode=::NewEnumerator0
    Override Seed=false
    Sphere Radius=0.1
    UseOwnerScale=true
   )
  ]
  particleUpdateScript [&ParticleState()]
  rendererClasses [&NiagaraSpriteRendererProperties]
  renderers [{rendererClass=&NiagaraSpriteRendererProperties}]
 }
 Stars {
  emitterUpdateScript [
   &EmitterState()
   &SpawnRate(Spawn Probability=0.8, SpawnRate=50, Use Spawn Probability=true)
   /SetVariables(Emitter.TwinkleCurve=DataInterface() {propertyValues {Curve {keys [<at 0 0.5>,<at 0.0726 0.7347>,<at 0.1854 0.4532>,<at 0.4085 1.0312>,<at 0.523 0.2103>,<at 0.6104 0.6655>,<at 0.731 0.0854>,<at 0.954 0.5224>,<at 1 0.5>]}}}) {isSetParameters=true}
  ]
  particleSpawnScript [
   &InitializeParticle(
    Color Channel Mode:ENiagaraChannelCorrelation=::NewEnumerator2
    Color Maximum=<rgba 2 1.5483 0.0174 1>
    Color Minimum=<rgba 1.5 0.1869 0.0339 1>
    Color Mode=::NewEnumerator2
    Sprite Size=&Multiply_Vector2D_ByFloat(Float=&RandomRangeFloat(Minimum=5, Maximum=15), Vector2D=<0.5 1>)
    Sprite Size Mode=::NewEnumerator4
    UsePositionOffset=false
    Write Material Random=true
   )
   &ShapeLocation(MaskSpawnGroup=false, Override Seed=false, Shape Origin=<0 0 0>, Sphere Radius=10, UseOwnerScale=true)
   &AddVelocity(
    Module.Use Falloff Distance=false
    Origin Offset Coordinate Space=::Simulation
    Override Seed=false
    Velocity Mode=::NewEnumerator1
    Velocity Origin=<0 0 0>
    Velocity Speed=&RandomRangeFloat(Minimum=25, Maximum=75)
   )
   &CameraOffset(Camera Offset Amount=-39)
  ]
  particleUpdateScript [
   &ParticleState()
   &Drag(Use Linear Drag=true, Use Rotational Drag=true)
   &SolveForcesAndVelocity(Clamp Velocity=false, Limit Acceleration=false)
   &ScaleColor(
    Scale Alpha=&FloatFromCurve(
     FloatCurve=DataInterface() {propertyValues {Curve {
        defaultValue=1
        keys [
         {arriveTangent=6.1295, interpMode=RCIM_Cubic, leaveTangent=6.1295, tangentMode=RCTM_User, time=0, value=0}
         {interpMode=RCIM_Cubic, tangentMode=RCTM_User, time=0.5, value=1}
        ]
       }}}
     Scale Curve=&Lerp_Float(
      A=0.01
      B=1
      Alpha=&FloatFromCurve(
       FloatCurve=$Emitter.TwinkleCurve
       Scale Curve=&FNTimeOfDayScalar(Night Scale=0.5, Morning Scale=0.75, Evening Scale=0.75)
      )
     )
    )
    ScaleA=true
    ScaleRGB=true
   )
   &ScaleSpriteSize(
    Uniform Curve Scale=&Lerp_Float(A=0.75, B=1, Alpha=&FloatFromCurve(FloatCurve=$Emitter.TwinkleCurve))
    Uniform Curve Sprite Scale=DataInterface() {propertyValues {Curve {keys [<at 0.5 1 RCIM_Cubic>, {arriveTangent=-5.2506, tangentMode=RCTM_User, time=1, value=0}]}}}
   )
   /SetVariables(Particles.PositionMod:NiagaraPosition=&AddVectorToPosition(
     Position=$Engine.Emitter.SimulationPosition
     Vector=&Add_Vector(
      A=&TransformVector(Vector=Linked() {linkedVariable {name=System.CenteredPos, type=NiagaraPosition}}, Destination Space=::Simulation)
      B=Linked() {linkedVariable {name=Particles.Position, type=NiagaraPosition}}
     )
    )) {isSetParameters=true}
  ]
  rendererClasses [&NiagaraSpriteRendererProperties]
  renderers [{rendererClass=&NiagaraSpriteRendererProperties}]
 }
 SunRays {
  emitterUpdateScript [
   &EmitterState()
   &SpawnBurst_Instantaneous(Spawn Count=11, Use Spawn Probability=false, UseLoopCountLimit=true)
  ]
  particleSpawnScript [
   &InitializeParticle(
    Color=<rgba 1 1 1 0.7>
    Lifetime=7
    Sprite Rotation Angle=&Multiply_FloatByInt(Float=&Divide_Float(A=360, B=11), Integer=&ReturnExecIndex())
    Sprite Rotation Mode=::NewEnumerator0
    Sprite Size=&Multiply_Vector2D(
     A=<15 55>
     B=&MakeCustomVector2DFromBool(Bool=$User.IsLargeBody, True Vector2D=<1.2 1.2>, False Vector2D=<1 1>)
    )
    Sprite Size Mode=::NewEnumerator4
    UsePositionOffset=false
    Write Material Random=true
   )
   &CameraOffset(Camera Offset Amount=-50) {module=CameraOffset001}
  ]
  particleUpdateScript [
   &ParticleState(Kill Particles When Lifetime Has Elapsed=false)
   &SpriteRotationRate(Rotation Rate=&Multiply_Float(A=10, B=&Sine(Normalized Angle=$Particles.NormalizedAge, Period=2, Scale=1, Bias=0)))
   &ScaleSpriteSize(
    Scale Factor=&MakeVector2D(
     X=&FloatFromCurve(FloatCurve=DataInterface() {propertyValues {Curve {keys [
          {interpMode=RCIM_Cubic, tangentMode=RCTM_User, time=0, value=1}
          {interpMode=RCIM_Cubic, tangentMode=RCTM_User, time=0.33, value=1.1}
          <at 0.66 1 RCIM_Cubic>
         ]}}})
     Y=&FloatFromCurve(
      FloatCurve=DataInterface() {propertyValues {Curve {keys [
          {interpMode=RCIM_Cubic, tangentMode=RCTM_User, time=0, value=0.75}
          {interpMode=RCIM_Cubic, tangentMode=RCTM_User, time=0.25, value=1.1}
          {tangentMode=RCTM_User, time=0.5, value=0.75}
         ]}}}
      CurveIndex=&Frac_Float(Float=&Add_Float(A=$Particles.NormalizedAge, B=&RandomRangeFloat(Minimum=0, Maximum=1)))
     )
    )
    Scale Sprite Size Mode=::NewEnumerator2
   ) {enabled=false}
   /SetVariables(Particles.Position=$System.CenteredPos) {isSetParameters=true, module=SetVariables_#4}
   &ScaleColor(
    Scale Alpha=&FNTimeOfDayScalar(Night Scale=0.8, Morning Scale=1, Evening Scale=1)
    ScaleA=true
    ScaleRGB=true
   )
  ]
  rendererClasses [&NiagaraSpriteRendererProperties]
  renderers [{rendererClass=&NiagaraSpriteRendererProperties}]
 }
}
summary {
 systemName=NS_Aura_FantasySunburst
 emitters [
  {emitterName=Backer, rendererClasses [&NiagaraSpriteRendererProperties]}
  {emitterName=SunRays, rendererClasses [&NiagaraSpriteRendererProperties]}
  {emitterName=Stars, rendererClasses [&NiagaraSpriteRendererProperties]}
  {bEnabled=false, emitterName=Minimal, rendererClasses [&NiagaraSpriteRendererProperties]}
  {emitterName=Beams, rendererClasses [&NiagaraSpriteRendererProperties]}
  {emitterName=BodyGlow, rendererClasses [&NiagaraSpriteRendererProperties]}
 ]
 userVariables [
  User.bIsFrontEnd {type=NiagaraBool}
  User.bIsFrontEndPreview {type=NiagaraBool}
  User.FrontEndScale {type=NiagaraFloat, defaultValue=0.75, description=""}
  User.IsFrontEnd {type=NiagaraBool}
  User.IsLargeBody {type=NiagaraBool}
 ]
}
